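# Two-sided, translucent lit material for soft smoke on particle beam trails, with per-vertex directional volumetric lighting.
# A panning grayscale smoke texture is offset and eroded through dynamic parameters (Erode, UOffset, PanSpeed).
# The eroded mask is multiplied by particle alpha, camera depth fade, MasterAlpha and a squared-U falloff that fades one trail end.
# A shading path switch adds scene depth fade on the default path and omits it on mobile.
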Material /BRCosmetics/Effects/Fort_Effects/Materials/Trails/M_Smoke_Trail_Soft_FadeStart_2Sided
graph {
 properties {
  bUsedWithBeamTrails=true
  blendMode=BLEND_Translucent
  pixelDepthOffsetMode=PDOM_Legacy
  translucencyLightingMode=TLM_VolumetricPerVertexDirectional
  twoSided=true
  usedShadingModels=Default Lit
 }
 nodes [
  n1 = ParticleColor()
  Reroute(None=@n1.RGB)
  n2 = DynamicParameter() {defaultValue=<rgba 1 0 1 1>, paramNames [Erode,UOffset,PanSpeed,MasterAlpha]}
  n3 = TextureCoordinate()
  n4 = AppendVector(A=@n2.UOffset, B=Constant())
  n5 = ScalarParameter() {defaultValue=0.6, parameterName=PanSpeed}
  n6 = Constant()
  n7 = Time()
  n8 = TextureSampleParameter2D(UVs=Panner(Coordinate=Add(A=@n3, B=@n4), Time=Multiply(A=@n2.PanSpeed, B=@n7), Speed=AppendVector(A=@n5, B=@n6)) {speedX=0.6}) {parameterName=TrailTexture_BChannel, samplerType=SAMPLERTYPE_Grayscale, texture=&T_Smoke_Trail_Soft}
  n9 = LinearInterpolate(Alpha=Add(A=@n2.Erode, B=@n8.G)) {constA=-3}
  n10 = Multiply(A=@n1.A, B=Clamp(None=Multiply(A=@n9, B=@n8.G)))
  n11 = MaterialFunctionCall(
   Fade Length=ScalarParameter() {defaultValue=512, parameterName=Fade Length}
   Fade Offset=ScalarParameter() {defaultValue=24, parameterName=Fade Offset}
  ) {materialFunction=&CameraDepthFade}
  n12 = ComponentMask(None=TextureCoordinate()) {r=true}
  n13 = Multiply(A=@n12, B=@n12)
  n14 = Multiply(
   A=Multiply(
    A=Multiply(A=@n10, B=@n11.Result)
    B=DynamicParameter() {defaultValue=<rgba 1 0 1 1>, paramNames [Erode,UOffset,PanSpeed,MasterAlpha]}.MasterAlpha
   )
   B=OneMinus(None=Clamp(None=@n13))
  )
  ShadingPathSwitch(Default=DepthFade(Opacity=@n14, FadeDistance=ScalarParameter() {defaultValue=25, parameterName=FadeDistance}), Mobile=@n14) {desc=Migrated from FeatureLevelSwitch}
 ]
}
