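# Unlit translucent trail material for Niagara ribbons and beam trails that draws soft smoke with two streak lines, tinted by particle color.
# Dynamic parameters drive the panning smoke texture's erosion, brightness, pan speed and master alpha.
# Zero-U-tiled sphere masks form the lines, broken up by panning noise and faded along U before joining the smoke in an emissive-adjustment function.
# Opacity combines particle alpha, eroded smoke, camera depth fade and a U-based fade at both ends.
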Material /BRCosmetics/Effects/Fort_Effects/Materials/Trails/M_Smoke_Trail_Soft_FadeStart_Paint
graph {
 properties {
  bUsedWithBeamTrails=true
  bUsedWithNiagaraRibbons=true
  blendMode=BLEND_Translucent
  pixelDepthOffsetMode=PDOM_Legacy
  shadingModel=MSM_Unlit
  usedShadingModels=Unlit
 }
 nodes [
  n1 = ParticleColor()
  n2 = DynamicParameter() {paramNames [Erode, EmissiveAmount, PanSpeed, MasterAlpha]}
  n3 = Time()
  n4 = TextureSampleParameter2D(UVs=Panner(Coordinate=TextureCoordinate() {coordinateIndex=1}, Time=Multiply(A=@n2.PanSpeed, B=@n3)) {speedX=0.6}) {parameterName=TrailTexture_BChannel, samplerType=SAMPLERTYPE_Grayscale, texture=&T_Smoke_Trail_Soft}
  n5 = Add(A=@n2.Erode, B=@n4.G)
  n6 = Clamp(None=Multiply(A=LinearInterpolate(Alpha=@n5) {constA=-3}, B=@n4.G))
  n7 = Multiply(A=DynamicParameter() {paramNames [Erode, EmissiveAmount, PanSpeed, MasterAlpha]}.EmissiveAmount, B=@n6)
  n8 = ComponentMask(None=TextureCoordinate()) {r=true}
  n9 = Clamp(None=Multiply(A=@n8, B=@n8))
  n10 = Multiply(A=OneMinus(None=@n9), B=@n9)
  n11 = TextureCoordinate() {uTiling=0}
  n12 = VectorParameter() {parameterName=Offset_Line1, defaultValue=<rgba 0 0.35 0 1>}
  n13 = ScalarParameter() {defaultValue=0.08, parameterName=Radius_Lines}
  n14 = ScalarParameter() {parameterName=Radius_Hardness}
  n15 = VectorParameter() {parameterName=Offset_Line2, defaultValue=<rgba 0 0.65 0 1>}
  n16 = Add(
   A=SphereMask(A=@n11, B=ComponentMask(None=@n12.RGB) {g=true, r=true}, Radius=@n13, Hardness=@n14)
   B=SphereMask(A=@n11, B=ComponentMask(None=@n15.RGB) {g=true, r=true}, Radius=@n13, Hardness=@n14)
  )
  n17 = TextureSample(UVs=Panner(Coordinate=TextureCoordinate() {uTiling=5}) {speedX=1}) {texture=&T_Smooth_tile}
  n18 = ComponentMask(None=TextureCoordinate()) {r=true}
  n19 = Multiply(A=@n18, B=@n18)
  n20 = Power(Base=@n19) {constExponent=0.1}
  n21 = Multiply(A=@n10, B=LinearInterpolate(A=Multiply(A=@n16, B=@n17.G), B=Constant(), Alpha=Clamp(None=@n20)))
  MaterialFunctionCall(Emissive=Add(A=Multiply(A=@n1.RGB, B=@n7), B=Multiply(A=@n1.RGB, B=@n21))) {materialFunction=&MF_FX_EmissiveAdjustments}
  Multiply(A=@n19, B=@n19)
  n22 = Multiply(A=@n1.A, B=@n6)
  n23 = MaterialFunctionCall(
   Fade Length=ScalarParameter() {defaultValue=512, parameterName=Fade Length}
   Fade Offset=ScalarParameter() {defaultValue=24, parameterName=Fade Offset}
  ) {materialFunction=&CameraDepthFade}
  Multiply(
   A=Multiply(A=Multiply(A=@n22, B=@n23.Result), B=DynamicParameter() {paramNames [Erode, EmissiveAmount, PanSpeed, MasterAlpha]}.MasterAlpha)
   B=@n10
  )
 ]
}
